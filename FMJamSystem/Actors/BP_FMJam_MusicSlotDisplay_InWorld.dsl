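; In-world display showing the title, artist and album art of the song playing in one Jam music slot, chosen by slot index.
; BeginPlay creates a dynamic album art material and binds to loop start/stop events on the Jam playspace's music manager.
; Starts in its slot look up the song catalog and request art asynchronously, toggling a material scalar by texture validity.
; Stopping the tracked loop instance or failed lookups clear the display.
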(event EventBeginPlay
  (bind _albumart (Variables|Default|GetAlbumArt))
  (bind _returnvalue (Rendering|Material|CreateDynamicMaterialInstance (Rendering|Material|GetMaterial _albumart 0) "None" "None" :!class /Script/Engine.KismetMaterialLibrary))
  (bind _output_get (Variables|Default|SetAlbumArtMaterial _returnvalue))
  (Rendering|Material|SetMaterial _albumart 0 _output_get)
  (CallFunction|ClearDisplay)
  (bind _returnvalue_2 (Class|PlayspaceLibrary|GetPlayspaceforActor self self))
  (bind _asjam_playspace (Utilities|Casting|CastToJamPlayspace _returnvalue_2)
    (:then
      _)
    (:CastFailed
      _))
  (if (not (Utilities|IsValid _asjam_playspace :!class /Script/Engine.KismetSystemLibrary))
    (Utilities|FlowControl|Delay 0.2))
  (Variables|Default|SetMusic%20Manager (Class|JamSessionInterface|GetMusicManager _asjam_playspace))
  (EventDispatchers|BindEventtoOnLoopStarted (Class|JamSessionInterface|GetMusicManager _asjam_playspace) (EventDispatchers|CreateEvent :!event Custom|OnLoopStarted))
  (EventDispatchers|BindEventtoOnLoopStopped (Class|JamSessionInterface|GetMusicManager _asjam_playspace) (EventDispatchers|CreateEvent :!event Custom|OnLoopStopped)))

(event Custom|OnLoopStarted
  (params (PlayParams :type Jam%20Play%20Params%20Structure%20%28by%20ref%29) (bChangedLoop :type Boolean))
  (bind (_songshortname _looptype _playernetid _loopinstanceid) (Utilities|Struct|BreakJamPlayParams PlayParams))
  (bind _relevant (CallFunction|IsRelevantSlot _loopinstanceid))
  (if _relevant
    (Variables|Default|SetCurrentlyPlayingLoopInstanceId _loopinstanceid)
    (CallFunction|UpdateDisplay _songshortname)))

(event Custom|OnLoopStopped
  (params (PlayParams :type Jam%20Play%20Params%20Structure%20%28by%20ref%29) (bChangedLoop :type Boolean))
  (bind (_songshortname _looptype _playernetid _loopinstanceid) (Utilities|Struct|BreakJamPlayParams PlayParams))
  (if (== _loopinstanceid (Variables|Default|GetCurrentlyPlayingLoopInstanceId))
    (CallFunction|ClearDisplay)))

(event Custom|UpdateDisplay
  (params (SongShortName :type Name))
  (if (!= SongShortName "")
    (bind _returnvalue (Class|SparksSongCatalog|GetSongbyShortName (GameInstanceSubsystems|GetSparksSongCatalog) (Utilities|String|ToString(Name) SongShortName)))
    (Utilities|IsValid _returnvalue
      (:Is%20Valid
        (bind _returnvalue_2 (Class|SparksSongItemDefinition|GetItemDeffromCatalogData _returnvalue))
        (Rendering|Components|TextRender|SetText (Variables|Default|GetTitle) (Class|FortMusicItem|GetTitle _returnvalue_2))
        (Rendering|Components|TextRender|SetText (Variables|Default|GetArtist) (Class|FortMusicItem|GetArtist _returnvalue_2))
        (SparksAlbumArt|RequestAlbumArt (GameInstanceSubsystems|GetSparksAlbumArtManager) SongShortName (delegate Custom|OnRequestFinished)))
      (:Is%20Not%20Valid
        (CallFunction|ClearDisplay)))
    (else
      (CallFunction|ClearDisplay))))

(event Custom|OnRequestFinished
  (params (SongShortName :type Name%20%28by%20ref%29) (AlbumArt :type Texture%202D%20Soft%20Object%20Reference%20%28by%20ref%29) (bWasSuccessful :type Boolean))
  (if bWasSuccessful
    (CallFunction|SetAlbumArtTexture AlbumArt)
    (else
      (CallFunction|ClearDisplay))))

(event Custom|ClearDisplay
  (Rendering|Components|TextRender|SetText (Variables|Default|GetTitle) (Variables|Default|GetEmptyText))
  (Rendering|Components|TextRender|SetText (Variables|Default|GetArtist) (Variables|Default|GetEmptyText))
  (CallFunction|SetAlbumArtTexture ""))

(fn IsRelevantSlot
  (params (LoopInstanceId :type Integer))
  (return (== (Mojo|JamSession|GetMusicSlotbyLoopInstanceId (Variables|Default|GetMusic%20Manager) LoopInstanceId) (Mojo|JamSession|GetMusicSlotbySlotIndex (Variables|Default|GetMusic%20Manager) (Variables|Default|GetMusicSlotIndex)))))

(fn SetAlbumArtTexture
  (params (AlbumArtTexture :type Texture%202D%20Soft%20Object%20Reference))
  (bind _output (Utilities|ResolveSoftReference AlbumArtTexture))
  (bind _albumartmaterial (Variables|Default|GetAlbumArtMaterial))
  (Rendering|Material|SetTextureParameterValue (Variables|Default|GetAlbumArtMaterial) "AlbumArt" _output)
  (Utilities|IsValid _output
    (:Is%20Valid
      (Rendering|Material|SetScalarParameterValue _albumartmaterial "DisplayAlbumArt" 1.0 :!class /Script/Engine.MaterialInstanceDynamic))
    (:Is%20Not%20Valid
      (Rendering|Material|SetScalarParameterValue _albumartmaterial "DisplayAlbumArt" 0.0 :!class /Script/Engine.MaterialInstanceDynamic))))

(fn ConstructionScript)
